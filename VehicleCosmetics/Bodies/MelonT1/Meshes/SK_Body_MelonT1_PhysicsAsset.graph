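# SK_Body_MelonT1_PhysicsAsset is the physics asset for the MelonT1 vehicle body skeletal mesh, part of the vehicle cosmetics content.
# It holds a single body on the root bone, with no collision shapes listed.
# It has no constraints, so it defines no jointed or ragdoll simulation between bones.
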PhysicsAsset /VehicleCosmetics/Bodies/MelonT1/Meshes/SK_Body_MelonT1_PhysicsAsset
bodies {root []}
constraints []
